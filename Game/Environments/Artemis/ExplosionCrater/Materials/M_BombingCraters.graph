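# Bomb crater surface material, Nanite-enabled, that blends crater meshes into the Apollo landscape.
# Height-lerps soil, rock and scree gravel layers using local height, with tone tints, a blast mask darkening the result and parallax occlusion gravel on Epic quality.
# An optional radial-gradient Helios sand layer goes on top.
# Landscape RVT color and height, plus UV1 radial masks, drive terrain blending, bypassed by Skip RVT and on mobile/forward paths.
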Material /Game/Environments/Artemis/ExplosionCrater/Materials/M_BombingCraters
graph {
 properties {
  bTangentSpaceNormal=false
  bUseMaterialAttributes=true
  bUsedWithNanite=true
  pixelDepthOffsetMode=PDOM_Legacy
  usedShadingModels=Default Lit
 }
 nodes [
  n1 = TextureCoordinate()
  AppendVector(A=@n1, B=Constant())
  Constant() {r=3}
  GetMaterialAttributes()
  n2 = ScalarParameter() {defaultValue=0.45, parameterName=ToneLerpTransitionMin}
  n3 = ScalarParameter() {defaultValue=1, parameterName=ToneLerpTransitionMax}
  n4 = NamedRerouteUsage()
  n5 = NamedRerouteUsage()
  n6 = ConstantBiasScale(None=DotProduct(A=Normalize(VectorInput=@n4), B=Normalize(VectorInput=@n5)))
  NamedRerouteDeclaration(None=Saturate(None=InverseLinearInterpolate(A=@n2, B=@n3, Value=@n6) {constA=0.45, constB=0.65})) {name=ToneLerpTransition, nodeColor=<rgba 1 0.2109 0 1>}
  n7 = MaterialFunctionCall!(Grass_UseCustomSlopeMask, Grass_CustomSlopeMask) {materialFunction=/Game/Athena/Helios/Environments/Landscape/Materials/MaterialFunctions_Utilities/MF_Helios_StaticMeshLandscapeLayerMatch}
  n8 = GetMaterialAttributes(None=@n7.SandLayer) {attributeGetTypes [#1]}
  n9 = Constant2Vector() {g=1, r=1}
  n10 = ScalarParameter() {defaultValue=5, group=1_Dunes, parameterName="[Not RVT] Normal Blend Value"}
  n11 = Normalize(VectorInput=Multiply(A=VertexNormalWS(), B=AppendVector(A=@n9, B=@n10)))
  n12 = ComponentMask(None=@n11) {g=true, r=true}
  n13 = ComponentMask(None=@n8.Normal) {g=true, r=true}
  n14 = Multiply(A=ComponentMask(None=@n11) {b=true}, B=ComponentMask(None=@n8.Normal) {b=true})
  n15 = Constant() {r=2}
  n16 = LinearInterpolate(A=AppendVector(A=Add(A=@n12, B=@n13), B=Multiply(A=@n14, B=@n15)), B=@n8.Normal) {constAlpha=0}
  n17 = GetMaterialAttributes(None=@n7.SandLayer) {attributeGetTypes [#1]}
  n18 = LinearInterpolate(A=@n17.Normal, B=Constant3Vector() {constant=<rgba 0 0 1 0>})
  NamedRerouteDeclaration(None=ShadingPathSwitch(
    Default=SetMaterialAttributes!(MaterialAttributes=@n8.MaterialAttributes, Normal=@n16) {attributeSetTypes [#1]}
    Mobile=SetMaterialAttributes(MaterialAttributes=@n17.MaterialAttributes, Normal=@n18) {attributeSetTypes [#1]}
   )) {name=Sand, nodeColor=<rgba 0 1 0.1562 1>}
  n19 = MaterialFunctionCall(Texture Scale=ScalarParameter() {defaultValue=0.6234, parameterName=ZoomBoom Innter Scale}) {materialFunction=&ScaleUVsByCenter}
  n20 = TextureSample(UVs=@n19.UVs) {
   samplerSource=SSM_Clamp_WorldGroupSettings
   samplerType=SAMPLERTYPE_LinearGrayscale
   texture=/Game/Environments/Artemis/ExplosionCrater/Textures/ZoomBoom_M
  }
  n21 = TextureSample(UVs=@n19.UVs) {
   samplerSource=SSM_Clamp_WorldGroupSettings
   samplerType=SAMPLERTYPE_Normal
   texture=/Game/Environments/Artemis/ExplosionCrater/Textures/ZoomBoom_N
  }
  n22 = Multiply(A=@n21.RGB, B=Constant3Vector() {constant=<rgba 2 2 0 0>})
  n23 = TextureSample(UVs=Multiply(A=NamedRerouteUsage(), B=Constant() {r=1})) {
   samplerSource=SSM_Wrap_WorldGroupSettings
   samplerType=SAMPLERTYPE_Normal
   texture=/Game/Environments/Apollo/Sets/Terrain/Textures/T_Apollo_Dirt_N_01
  }
  n24 = Multiply(A=@n23.RGB, B=Constant3Vector() {constant=<rgba 2 2 1 0>})
  n25 = SetMaterialAttributes!(
   MaterialAttributes
   Base Color=LinearInterpolate(
    A=VectorParameter() {parameterName=SoilI Inside, defaultValue=<rgba 0.3038 0.3038 0.3038 0>}.RGB
    B=VectorParameter() {parameterName=Soil Edges, defaultValue=<rgba 0.1997 0.1927 0.1889 1>}.RGB
    Alpha=@n20.R
   )
   Specular=Constant() {r=0.1}
   Roughness=Constant() {r=1}
   Normal=Transform(None=Add(A=@n22, B=@n24))
   Opacity=Constant()
  ) {attributeSetTypes [
    #2
    #3
    #4
    #1
    #5
   ]}
  NamedRerouteDeclaration(None=@n25) {name=BottomSoil, nodeColor=<rgba 0.3359 0 1 1>}
  n26 = NamedRerouteUsage()
  n27 = Add(A=Multiply(A=TextureCoordinate(), B=Constant() {r=4}), B=Constant() {r=0.5})
  n28 = TextureSample(UVs=@n27) {samplerSource=SSM_Wrap_WorldGroupSettings, texture=/Game/Environments/Apollo/Sets/Terrain/Textures/T_Apollo_DirtHeight_D}
  n29 = TextureSample(UVs=@n27) {
   samplerSource=SSM_Wrap_WorldGroupSettings
   samplerType=SAMPLERTYPE_Normal
   texture=/Game/Environments/Apollo/Sets/Terrain/Textures/T_Apollo_DirtHeight_N
  }
  n30 = Transform(None=@n29.RGB)
  n31 = NamedRerouteDeclaration(None=@n28.A) {name=HM, nodeColor=<rgba 1 0 0.3984 1>}
  n32 = SetMaterialAttributes(
   Base Color=@n28.R
   Specular=Multiply(A=Fresnel(Normal=@n30) {baseReflectFraction=0.2, exponent=2}, B=ScalarParameter() {defaultValue=0.5, parameterName=Rocks S})
   Roughness=LinearInterpolate(
    A=ScalarParameter() {defaultValue=2, parameterName=Rocks R Min}
    B=ScalarParameter() {defaultValue=0.6, parameterName=Rocks R Max}
    Alpha=@n31
   )
   Normal=@n30
   Opacity=@n31
  ) {attributeSetTypes [
    #2
    #3
    #4
    #1
    #5
   ]}
  n33 = MaterialFunctionCall(
   Transition Phase=ScalarParameter() {defaultValue=0.4053, parameterName=Rocks Blend}
   Height Texture=@n28.A
   Contrast=ScalarParameter() {defaultValue=1, parameterName=Rocks Contrast}
  ) {materialFunction=&HeightLerp}
  n34 = NamedRerouteUsage()
  n35 = ObjectRadius()
  n36 = NamedRerouteDeclaration(None=Multiply(A=TextureCoordinate(), B=Divide(A=@n35) {constB=128})) {name=ObjectScale, nodeColor=<rgba 1 0.5859 0 1>}
  n37 = Divide(A=@n36, B=ScalarParameter() {defaultValue=1.5, parameterName=GravelTextureSize})
  n38 = TextureObject() {texture=/Game/Environments/Apollo/Sets/Terrain/Textures/T_Apollo_Scree_D_01}
  n39 = ScalarParameter() {defaultValue=0.025, group=POM, parameterName=Gravel Depth}
  n40 = MaterialFunctionCall!(World Position=WorldPosition().XYZ) {materialFunction=/Engine/Functions/Engine_MaterialFunctions02/Utility/DepthFromWorldPosition}
  n41 = Constant() {r=2500}
  n42 = Saturate(None=Divide(A=Subtract(A=@n40.Result, B=@n41), B=Constant() {r=-2500}))
  n43 = Constant() {r=1}
  n44 = Add(A=Multiply(A=Constant() {r=4}, B=@n42), B=@n43)
  n45 = Constant() {r=16}
  n46 = LinearInterpolate(Alpha=MaterialFunctionCall() {materialFunction=&DitherTemporalAA}.Result) {constA=1, constB=1.5}
  n47 = Add(A=@n43, B=Multiply(A=@n42, B=Multiply(A=@n45, B=@n46)))
  n48 = Constant() {r=0.5}
  n49 = QualitySwitch(
   Default=@n37
   Epic=MaterialFunctionCall!(
    Heightmap Texture=@n38
    Height Ratio=@n39
    Min Steps=@n44
    Max Steps=@n47
    UVs=@n37
    Heightmap Channel
    Reference Plane=@n48
    -----------------------------
    Use World Coordinates
    Specify Manual Texture Size
    Manual Texture Size
    ------------------------------
    "Render Shadows (Occlusion Mapping)"
    Light Vector
    Shadow Steps
    Shadow Penumbra
    Transform To VertexNormal
   ) {materialFunction=/Engine/Functions/Engine_MaterialFunctions01/Texturing/ParallaxOcclusionMapping}.Parallax UVs
  )
  n50 = TextureSample(UVs=@n49, Tex=@n38) {samplerSource=SSM_Wrap_WorldGroupSettings}
  n51 = TextureSample(UVs=@n49) {
   samplerSource=SSM_Wrap_WorldGroupSettings
   samplerType=SAMPLERTYPE_Masks
   texture=/Game/Environments/Apollo/Sets/Terrain/Textures/T_Apollo_Scree_S_01
  }
  n52 = Constant3Vector() {constant=<rgba 0 0 1 0>}
  n53 = TextureSampleParameter2D(UVs=@n49) {
   group=Base
   parameterName=Normal
   samplerSource=SSM_Wrap_WorldGroupSettings
   samplerType=SAMPLERTYPE_Normal
   sortPriority=2
   texture=/Game/Environments/Apollo/Sets/Terrain/Textures/T_Apollo_Scree_N_01
  }
  n54 = ScalarParameter() {defaultValue=1, parameterName=GravelNormalStrength, sliderMax=1}
  n55 = Divide(A=@n36, B=ScalarParameter() {defaultValue=20, parameterName=GravelNoiseTextureSize})
  n56 = TextureObject() {samplerType=SAMPLERTYPE_LinearGrayscale, texture=/Game/Weapons/FORT_Melee/Pickaxe_SnowFall_Female/FX/Snowfall_pick_Noise}
  n57 = MaterialFunctionCall(Heightmap Texture=@n56, Height Ratio=@n39, Min Steps=@n44, Max Steps=@n47, UVs=@n55, Reference Plane=@n48) {materialFunction=&ParallaxOcclusionMapping}
  n58 = LinearInterpolate(
   A=ScalarParameter() {defaultValue=-0.2, parameterName=GravelNoiseMin}
   B=ScalarParameter() {defaultValue=1, parameterName=GravelNoiseMax}
   Alpha=TextureSample(UVs=QualitySwitch(Default=@n55, Epic=@n57.Parallax UVs), Tex=@n56) {samplerSource=SSM_Wrap_WorldGroupSettings}.R
  ) {constA=0.5}
  n59 = ScalarParameter() {defaultValue=0.8, parameterName=GravelNoisePower}
  n60 = LinearInterpolate(B=Saturate(None=Power(Base=@n58, Exp=@n59)), Alpha=ScalarParameter() {parameterName=UseGravelNoise, sliderMax=1}) {constA=1}
  n61 = NamedRerouteDeclaration(None=ComponentMask(None=Multiply(A=@n50.RGBA, B=@n60)) {a=true}) {name=Tiled H, nodeColor=<rgba 1 0.6563 0 1>}
  n62 = SetMaterialAttributes(Base Color=@n50.G, Specular=@n51.R, Roughness=@n51.R, Normal=Transform(None=LinearInterpolate(A=@n52, B=@n53.RGB, Alpha=@n54)), Opacity=@n61) {attributeSetTypes [
    #2
    #3
    #4
    #1
    #5
   ]}
  n63 = MaterialFunctionCall(Transition Phase=NamedRerouteUsage(), Height Texture=@n61) {materialFunction=&HeightLerp}
  n64 = WorldPosition()
  n65 = ActorPositionWS()
  n66 = Transform(None=NamedRerouteDeclaration(None=Subtract(A=@n64.XYZ, B=@n65)) {name=Actor Space, nodeColor=<rgba 0 1 0.8125 1>}) {transformSourceType=TRANSFORMSOURCE_World, transformType=TRANSFORM_Local}
  n67 = Divide(
   A=Add(A=ComponentMask(None=@n66) {b=true}, B=ScalarParameter() {defaultValue=75, parameterName=GOR})
   B=ScalarParameter() {defaultValue=250, parameterName=Gravel Over Rocks Blur}
  )
  n68 = MaterialFunctionCall(
   Transition Phase=LinearInterpolate(
    A=ScalarParameter() {defaultValue=0.6, parameterName=GravelDensityWithoutHeight}
    B=OneMinus(None=@n67)
    Alpha=ScalarParameter() {defaultValue=1, parameterName=GravelHeightLerp}
   )
   Height Texture=NamedRerouteUsage()
   Contrast=Multiply(A=NamedRerouteUsage(), B=Constant() {r=15})
  ) {materialFunction=&HeightLerp}
  n69 = GetMaterialAttributes(None=BlendMaterialAttributes(
    A=BlendMaterialAttributes(A=@n26, B=@n32, Alpha=@n33.Alpha)
    B=BlendMaterialAttributes(A=@n34, B=@n62, Alpha=@n63.Lerp Alpha No Contrast) {vertexAttributeBlendType=UseB}
    Alpha=@n68.Alpha
   ) {vertexAttributeBlendType=UseB}) {attributeGetTypes [
    #2
    #5
    #1
    #3
   ]}
  NamedRerouteDeclaration(None=@n69.Normal) {name=Mixed Normals, nodeColor=<rgba 0.9688 1 0 1>}
  n70 = DotProduct(A=NamedRerouteUsage(), B=NamedRerouteUsage())
  n71 = ScalarParameter() {defaultValue=1000}
  NamedRerouteDeclaration(None=Saturate(None=Divide(A=@n70, B=@n71))) {name="Central \"Ligthing\"", nodeColor=<rgba 1 0.2109 0 1>}
  n72 = AppendVector(
   A=ScalarParameter() {defaultValue=0.5, group=Sand, parameterName=Sand_X}
   B=ScalarParameter() {defaultValue=0.5, group=Sand, parameterName=Sand_Y}
  )
  n73 = ScalarParameter() {defaultValue=0.1, group=Sand, parameterName=sand layer}
  n74 = Saturate(None=Multiply(
    A=MaterialFunctionCall(CenterPosition=@n72, Radius=@n73) {materialFunction=&RadialGradientExponential}.RadialGradientExponential
    B=ScalarParameter() {defaultValue=5, group=Sand, parameterName=sand_Mutli}
   ) {constB=5})
  n75 = MaterialFunctionCall(
   CenterPosition=AppendVector(
    A=ScalarParameter() {defaultValue=0.5, group=Sand, parameterName=Sand_Cutout_X}
    B=ScalarParameter() {defaultValue=0.5, group=Sand, parameterName=Sand_Cutout_Y}
   )
   Radius=ScalarParameter() {defaultValue=0.1, group=Sand, parameterName=sand innercutout}
  ) {materialFunction=&RadialGradientExponential}
  n76 = ScalarParameter() {defaultValue=5, group=Sand, parameterName=sand_Cutout_Mutli}
  n77 = Subtract(A=@n74, B=Saturate(None=Multiply(A=@n75.RadialGradientExponential, B=@n76) {constB=5}))
  NamedRerouteDeclaration(None=StaticSwitchParameter(True=Saturate(None=@n77), False=@n74) {parameterName=Sand_Inner_Cutout}) {name=Sand_Gradient, nodeColor=<rgba 0.4766 0 1 1>}
  NamedRerouteUsage()
  n78 = RuntimeVirtualTextureSampleParameter() {
   group=RVT
   materialType=BaseColor_Normal_Specular_YCoCg
   parameterName=RVT
   sortPriority=3
   virtualTexture=/Game/Athena/Apollo/Maps/Landscape/VT/RVT_Apollo_Landscape
  }
  n79 = SetMaterialAttributes!(MaterialAttributes, Base Color=@n78.BaseColor, Specular=@n78.BaseColor, Roughness=@n78.BaseColor, Normal=@n78.BaseColor) {attributeSetTypes [
    #2
    #3
    #4
    #1
   ]}
  n80 = VectorParameter() {parameterName=GOT, defaultValue=<rgba 0.402 0.305 0.2232 0>}
  n81 = SphereMask(
   A=Subtract(A=@n1) {constB=0.5}
   B=ComponentMask(None=VectorParameter() {parameterName=ToneLerpSphereCenter, defaultValue=<rgba -0.015 0.015 0 0>}.RGB) {g=true, r=true}
   Radius=ScalarParameter() {defaultValue=0.35, parameterName=ToneLerpSphereRadius}
   Hardness=ScalarParameter() {defaultValue=0.6, parameterName=ToneLerpSphereHardness}
  ) {attenuationRadius=0.35, hardnessPercent=65}
  n82 = OneMinus(None=@n81)
  n83 = Subtract(A=@n82) {constB=0.5}
  n84 = ConstantBiasScale(None=@n81) {bias=-0.5, scale=2}
  n85 = LinearInterpolate(A=NamedRerouteUsage(), B=Saturate(None=Ceil(None=@n83)), Alpha=Saturate(None=Abs(None=@n84)))
  n86 = ScalarParameter() {parameterName=UseToneLerp, sliderMax=1}
  n87 = LinearInterpolate(
   A=@n80.RGB
   B=LinearInterpolate(A=@n80.RGB, B=VectorParameter() {parameterName=LGOT, defaultValue=<rgba 1 0.4624 0.224 1>}.RGB, Alpha=@n85)
   Alpha=@n86
  )
  n88 = Multiply(B=@n87) {constA=0.65}
  n89 = Constant() {r=0.5}
  n90 = NamedRerouteUsage()
  n91 = VectorParameter() {parameterName=GT, defaultValue=<rgba 0.3647 0.3451 0.2941 0>}
  n92 = VectorParameter() {parameterName=LGT, defaultValue=<rgba 1 0.6305 0.349 1>}
  n93 = LinearInterpolate(
   A=LinearInterpolate(A=@n91.RGB, B=LinearInterpolate(A=@n91.RGB, B=@n92.RGB, Alpha=@n85), Alpha=@n86)
   B=LinearInterpolate(
    A=VectorParameter() {parameterName=GGT, defaultValue=<rgba 1 0.6604 0.1499 1>}.RGB
    B=VectorParameter() {parameterName=GGOT, defaultValue=<rgba 1 0.4635 0.1892 1>}.RGB
    Alpha=@n82
   )
   Alpha=LinearInterpolate(B=Saturate(None=@n68.Alpha), Alpha=ScalarParameter() {parameterName=UseGravelTone, sliderMax=1})
  ) {constAlpha=1}
  n94 = Multiply(B=@n93) {constA=0.55}
  n95 = MaterialFunctionCall(
   A=LinearInterpolate(A=Desaturation(None=@n88, Fraction=@n89), B=@n87, Alpha=@n90)
   B=LinearInterpolate(A=Desaturation(None=@n94, Fraction=@n89), B=@n93, Alpha=@n90)
   Transition Phase=ScalarParameter() {defaultValue=0.4, parameterName=Global Rock to Soil}
   Height Texture=@n69.Opacity
   Contrast=Constant()
  ) {materialFunction=&HeightLerp}
  n96 = Multiply(A=@n69.Specular, B=Constant() {r=0.25})
  n97 = NamedRerouteUsage()
  n98 = GetMaterialAttributes(None=SetMaterialAttributes!(
    MaterialAttributes=@n69.MaterialAttributes
    Base Color=Multiply(A=@n69.Base Color, B=@n95.Results)
    Specular=LinearInterpolate(A=@n96, B=@n69.Specular, Alpha=@n97)
   ) {attributeSetTypes [#2, #3]}) {attributeGetTypes [
    #2
    #3
    #4
    #1
   ]}
  n99 = TextureSample(UVs=MaterialFunctionCall(Texture Scale=ScalarParameter() {defaultValue=1.0534, desc=BP, parameterName=Boom Scale}) {materialFunction=&ScaleUVsByCenter}.UVs) {
   samplerSource=SSM_Clamp_WorldGroupSettings
   samplerType=SAMPLERTYPE_LinearGrayscale
   texture=&ZoomBoom_M
  }
  n100 = NamedRerouteDeclaration(None=@n99.R) {name=BoomZoom, nodeColor=<rgba 1 0.1875 0 1>}
  n101 = Max(A=@n100, B=ScalarParameter() {defaultValue=0.5, parameterName=Boom Min Value})
  n102 = SetMaterialAttributes!(
   MaterialAttributes=@n98.MaterialAttributes
   Base Color=Multiply(A=@n98.Base Color, B=@n101)
   Specular=Multiply(A=@n98.Specular, B=@n101)
   Roughness=LinearInterpolate(A=ScalarParameter() {defaultValue=0.65, desc=BP, parameterName=Boom R}, B=@n98.Roughness, Alpha=@n100)
   Normal=@n98.MaterialAttributes
  ) {attributeSetTypes [
    #2
    #3
    #4
    #1
   ]}
  n103 = ScalarParameter() {defaultValue=120, desc=BP, group=RVT, parameterName=Blend RVT, sortPriority=0}
  n104 = WorldPosition()
  n105 = RuntimeVirtualTextureSampleParameter() {
   group=RVT
   materialType=WorldHeight
   parameterName=RVTHeight
   sortPriority=4
   virtualTexture=/Game/Athena/Apollo/Maps/Landscape/VT/RVT_Apollo_Landscape_Height
  }
  n106 = NamedRerouteDeclaration(None=Subtract(A=ComponentMask(None=@n104.XYZ) {b=true}, B=@n105.WorldHeight)) {name=RVT Height, nodeColor=<rgba 0.2187 1 0 1>}
  n107 = BlendMaterialAttributes(
   A=@n79
   B=NamedRerouteDeclaration(None=@n102) {name=Pre RVT, nodeColor=<rgba 1 0 0.75 1>}
   Alpha=MaterialFunctionCall(
    Transition Phase=Divide(A=@n106, B=@n103)
    Height Texture=NamedRerouteUsage()
    Contrast=ScalarParameter() {defaultValue=0.5, group=RVT, parameterName=Blend RVT Contrast, sortPriority=1}
   ) {materialFunction=&HeightLerp}.Alpha
  ) {vertexAttributeBlendType=UseA}
  n108 = NamedRerouteUsage()
  n109 = TextureCoordinate() {coordinateIndex=1}
  n110 = MaterialFunctionCall(UVs=@n109, Radius=Constant() {r=0.3}, Density=ScalarParameter() {defaultValue=3.5, parameterName=InnerBlendRVTDensity}) {materialFunction=&RadialGradientExponential}
  n111 = Min(
   A=MaterialFunctionCall(UVs=@n109, Radius=Constant() {r=0.3507}, Density=Constant() {r=3}) {materialFunction=&RadialGradientExponential}.RadialGradientExponential
   B=LinearInterpolate(B=OneMinus(None=@n110.RadialGradientExponential), Alpha=ScalarParameter() {parameterName=UseInnerBlendRVT}) {constA=1}
  )
  n112 = NamedRerouteUsage()
  n113 = StaticSwitchParameter(
   True=BlendMaterialAttributes(A=NamedRerouteUsage(), B=@n112, Alpha=NamedRerouteUsage()) {desc=helios addition for 31.0, vertexAttributeBlendType=UseB}
   False=@n112
  ) {group=Sand, parameterName=UseSand}
  n114 = GetMaterialAttributes(None=@n113) {attributeGetTypes [#2]}
  n115 = StaticSwitchParameter(True=@n114.Base Color, False=Multiply(A=@n114.Base Color) {constB=0.5}) {group=Sand, parameterName=UseSand}
  n116 = MaterialFunctionCall(
   NoRVT=@n113
   RVT=MaterialProxyReplace(
    Realtime=BlendMaterialAttributes(A=@n107, B=@n108, Alpha=@n111) {vertexAttributeBlendType=UseA}
    MaterialProxy=SetMaterialAttributes!(MaterialAttributes=@n114.MaterialAttributes, Base Color=@n115) {attributeSetTypes [#2]}
   )
  ) {materialFunction=&FN_RVT_Switch_Material}
  StaticSwitchParameter(True=@n113, False=ShadingPathSwitch(Default=@n116.Result, Deferred=@n116.Result, Forward=@n113, Mobile=@n113)) {group=RVT, parameterName=Skip RVT, sortPriority=-4}
  Subtract(A=@n13, B=@n12)
 ]
}
