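# Translucent crater decal material for the Artemis explosion crater, set up to write into a runtime virtual texture.
# A scaled, rotatable grayscale mask drives HeightLerp blends of two colors, roughness and specular.
# Opacity multiplies a radial gradient by a height blend of the inverted mask.
# Material attributes pass through a normal transform, then feed the RVT output and a virtual-texture feature switch selecting RVT or non-RVT results.
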Material /Game/Environments/Artemis/ExplosionCrater/Materials/M_RVT_Crater_Decal_BP
graph {
 properties {
  bAutomaticallySetUsageInEditor=false
  bCastRayTracedShadows=false
  bTangentSpaceNormal=false
  bUseMaterialAttributes=true
  bUsedWithInstancedStaticMeshes=true
  blendMode=BLEND_Translucent
  pixelDepthOffsetMode=PDOM_Legacy
  usedShadingModels=Default Lit
 }
 nodes [
  MaterialFunctionCall!(RVT, ByPass RVT) {materialFunction=/Game/Athena/Apollo/Maps/Landscape/VT/MaterialFunctions/EnableRVT}
  n1 = ScalarParameter() {defaultValue=2, parameterName=Scale}
  n2 = TextureSample(UVs=Rotator(
    Coordinate=MaterialFunctionCall(Texture Scale=@n1) {materialFunction=&ScaleUVsByCenter}.UVs
    Time=ScalarParameter() {defaultValue=1.8636, parameterName=RotateDecal}
   )) {samplerType=SAMPLERTYPE_LinearGrayscale, texture=/Game/Environments/Artemis/ExplosionCrater/Textures/ZoomBoom_M}
  n3 = MaterialFunctionCall(
   A=VectorParameter() {parameterName=C1, defaultValue=<rgba 1 0 0 0>}.RGB
   B=VectorParameter() {parameterName=C2, defaultValue=<rgba 0.5469 0.3097 0.0922 0>}.RGB
   Transition Phase=Constant() {r=0.4}
   Height Texture=@n2.R
   Contrast=Constant()
  ) {materialFunction=&HeightLerp}
  n4 = LinearInterpolate(A=ScalarParameter() {defaultValue=0.55, parameterName=R}, Alpha=@n3.Alpha)
  n5 = LinearInterpolate(A=ScalarParameter() {defaultValue=0.15, parameterName=S}, Alpha=@n3.Alpha) {constB=0.5}
  n6 = MaterialFunctionCall(Radius=Constant() {r=0.5}, Density=Constant() {r=7}) {materialFunction=&RadialGradientExponential}
  n7 = MaterialFunctionCall(Transition Phase=ScalarParameter() {defaultValue=0.6, parameterName=T}, Height Texture=OneMinus(None=@n2.R), Contrast=Constant() {r=1}) {materialFunction=&HeightLerp}
  n8 = GetMaterialAttributes(None=SetMaterialAttributes!(
    MaterialAttributes=SetMaterialAttributes!(MaterialAttributes, Base Color=@n3.Results, Roughness=@n4, Specular=@n5) {attributeSetTypes [#1, #2, #3]}
    Opacity=Multiply(A=@n6.RadialGradientExponential, B=@n7.Alpha)
   ) {attributeSetTypes [#4]}) {attributeGetTypes [#5]}
  n9 = SetMaterialAttributes!(MaterialAttributes=@n8.MaterialAttributes, Normal=Transform(None=@n8.Normal)) {attributeSetTypes [#5]}
  n10 = GetMaterialAttributes(None=@n9) {attributeGetTypes [
    #1
    #3
    #2
    #5
    #4
   ]}
  RuntimeVirtualTextureOutput!(
   BaseColor=@n10.Base Color
   Specular=@n10.Base Color
   Roughness=@n10.Base Color
   Normal
   WorldHeight
   Opacity=@n10.Base Color
   Mask
   Displacement
   Mask4
  )
  n11 = VirtualTextureFeatureSwitch!(
   No=@n9
   Yes=MaterialFunctionCall(NoRVT=@n9, RVT=@n10.MaterialAttributes) {materialFunction=&FN_RVT_Switch_Material}.Result
  )
  ShadingPathSwitch(Default=@n11, Deferred=@n11, Forward=@n11, Mobile=@n11)
 ]
}
