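; Configurable bomb crater environment actor whose look is built entirely in the construction script; its event graph stubs are empty.
; It assigns the crater mesh and creates a dynamic material instance bound to color and height Runtime Virtual Textures for blending.
; Exposed variables drive crater and gravel tones, noise, center shape and RVT blend, plus a sand layer mask marked as a Helios addition.
; The ground-cover smoke component is destroyed when smoke VFX is disabled.
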(event EventBeginPlay)

(event Collision|EventActorBeginOverlap
  (params (OtherActor :type Actor%20Object%20Reference)))

(event EventTick
  (params (DeltaSeconds :type Float%20%28single-precision%29)))

(fn ConstructionScript
  (bind _bombcrater (Variables|Default|GetBombCrater))
  (bind _bombcrater_2 (Variables|Default|GetBombCrater))
  (Components|StaticMesh|SetStaticMesh (Variables|Default|GetBombCrater) (Variables|Default|GetMesh))
  (bind _returnvalue (Rendering|Material|CreateDynamicMaterialInstance (Variables|Default|GetBombCrater) 0 (Variables|Default|GetSource%20Material) "None" :!class /Script/Engine.PrimitiveComponent))
  (Rendering|Material|SetRuntimeVirtualTextureParameterValue _returnvalue "RVTHeight" (Variables|Default|GetRVTHeight))
  (Rendering|Material|SetRuntimeVirtualTextureParameterValue _returnvalue "RVT" (Variables|Default|GetRVT))
  (Rendering|Material|SetVectorParameterValueonMaterials _bombcrater "GT" (Math|Conversions|ToVector(LinearColor) (Variables|Default|GetCrater%20Tone)))
  (Rendering|Material|SetVectorParameterValueonMaterials _bombcrater "GOT" (Math|Conversions|ToVector(LinearColor) (Variables|Default|GetCrater%20Tone%20Outside)))
  (Rendering|Material|SetScalarParameterValueonMaterials _bombcrater "UseToneLerp" (Math|Float|SelectFloat 1.0 0.0 (Variables|Default|GetUse%20Crater%20Tone%20Lerp)))
  (Rendering|Material|SetVectorParameterValueonMaterials _bombcrater "LGT" (Math|Conversions|ToVector(LinearColor) (Variables|Default|GetCrater%20Tone%20Lerp)))
  (Rendering|Material|SetVectorParameterValueonMaterials _bombcrater "LGOT" (Math|Conversions|ToVector(LinearColor) (Variables|Default|GetCrater%20Tone%20Outside%20Lerp)))
  (Rendering|Material|SetScalarParameterValueonMaterials _bombcrater "UseGravelTone" (Math|Float|SelectFloat 1.0 0.0 (Variables|Default|GetUse%20Gravel%20Tone)))
  (Rendering|Material|SetVectorParameterValueonMaterials _bombcrater "GGT" (Math|Conversions|ToVector(LinearColor) (Variables|Default|GetGravel%20Tone)))
  (Rendering|Material|SetVectorParameterValueonMaterials _bombcrater "GGOT" (Math|Conversions|ToVector(LinearColor) (Variables|Default|GetGravel%20Tone%20Outside)))
  (Rendering|Material|SetScalarParameterValueonMaterials _bombcrater "UseGravelNoise" (Math|Float|SelectFloat 1.0 0.0 (Variables|Default|GetUse%20Gravel%20Noise)))
  (Rendering|Material|SetScalarParameterValueonMaterials _bombcrater "GravelNormalStrength" (Variables|Default|GetGravel%20Normal%20Strength))
  (Rendering|Material|SetScalarParameterValueonMaterials _bombcrater "GravelHeightLerp" (Variables|Default|GetGravel%20Height%20Lerp))
  (Rendering|Material|SetScalarParameterValueonMaterials _bombcrater "GOR" (Math|Float|MapRangeClamped (Variables|Default|GetCrater%20Gravel%20On%20Rocks) 0.0 1.0 9e+01 55.0))
  (Rendering|Material|SetScalarParameterValueonMaterials _bombcrater "Boom Scale" (Variables|Default|GetCrater%20Center%20Scale))
  (Rendering|Material|SetScalarParameterValueonMaterials _bombcrater "Boom R" (Variables|Default|GetCrater%20Center%20Roughness))
  (Rendering|Material|SetScalarParameterValueonMaterials _bombcrater "Blend RVT" (Variables|Default|GetCrater%20Blend%20RVT))
  (Rendering|Material|SetScalarParameterValueonMaterials _bombcrater "UseInnerBlendRVT" (Math|Float|SelectFloat 1.0 0.0 (Variables|Default|GetUse%20Inner%20Blend%20RVT)))
  (Rendering|Material|SetScalarParameterValueonMaterials _bombcrater "InnerBlendRVTDensity" (Variables|Default|GetInner%20Blend%20RVT%20Dentsity))
  ;[0.7656 0.5184 0.5184 0.5] Addition for helios
  (Rendering|Material|SetScalarParameterValueonMaterials _bombcrater_2 "sand layer" (Variables|Default|GetSand_Radius))
  (Rendering|Material|SetScalarParameterValueonMaterials _bombcrater_2 "Sand_X" (Variables|Default|GetSand%20X%20position))
  (Rendering|Material|SetScalarParameterValueonMaterials _bombcrater_2 "Sand_Y" (Variables|Default|GetSand%20Y%20Position))
  (Rendering|Material|SetScalarParameterValueonMaterials _bombcrater_2 "sand_Mutli" (Variables|Default|GetSand_Mask_Multi))
  (if (not (Variables|Default|GetUse%20Smoke%20VFX))
    (Components|DestroyComponent (Variables|Default|GetBP_Smoke_CrashSite_GroundCover))))
